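# Control Rig for the SageTwig medium male player body, correcting deformation and moving armor and gear bones with the pose.
# A sequence runs the shared DeformRig function and weighted parent constraints for attach points, pouches, holster and straps.
# Spherical pose readers on arms and thighs weight corrective offsets, with left shoulderpad offsets also scaled by bone distance.
# A capsule clamp constrains right pauldron bones; a spline-driven strap-pin setup is left unconnected.
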ControlRigBlueprint /BRCosmetics/Characters/Player/Male/Medium/Bodies/M_MED_SageTwig/Meshes/M_MED_SageTwig_ControlRig
graphs [
 RigVMFunctionLibrary {links [], nodes []}
 RigVMFunction_Sequence_1_ContainedGraph {nodes [
   n1 = RigVMFunctionEntryNode()
   RigVMFunction_Sequence = &RigVMFunction_Sequence(ExecuteContext=@n1.ExecuteContext)
   n2 = &RigVMFunction_Sequence(ExecuteContext=@RigVMFunction_Sequence.B)
   n3 = &RigVMFunction_Sequence(ExecuteContext=@n2.B)
   n4 = &RigVMFunction_Sequence(ExecuteContext=@n3.B)
   n5 = &RigVMFunction_Sequence(ExecuteContext=@n4.B)
   n6 = &RigVMFunction_Sequence(ExecuteContext=@n5.B)
   n7 = &RigVMFunction_Sequence(ExecuteContext=@n6.B)
   n8 = &RigVMFunction_Sequence(ExecuteContext=@n7.B)
   n9 = &RigVMFunction_Sequence(ExecuteContext=@n8.B)
   n10 = &RigVMFunction_Sequence(ExecuteContext=@n9.B)
   n11 = &RigVMFunction_Sequence(ExecuteContext=@n10.B)
   RigVMFunctionReturnNode(A=@RigVMFunction_Sequence.A, B=@n2.A, K=@n11.A, L=@n11.B, C=@n3.A, D=@n4.A, E=@n5.A, F=@n6.A, G=@n7.A, H=@n8.A, I=@n9.A, J=@n10.A)
  ]}
 RigVMModel {nodes [
   &RigUnit_ParentConstraint(
    AdvancedSettings=(RotationOrderForFilter=XZY)
    Child=(Type=Bone,Name="dyn_pauldron_fr_01_r_MSTR")
    ExecutePin=()
    Filter=(TranslationFilter=(bX=True,bY=True,bZ=True),RotationFilter=(bX=false,bY=false,bZ=false),ScaleFilter=(bX=True,bY=True,bZ=True))
    Parents=((Item=(Type=Bone,Name="neck_01"),Weight=0.0552),(Item=(Type=Bone,Name="spine_04"),Weight=0.0762),(Item=(Type=Bone,Name="clavicle_r"),Weight=0.2033),(Item=(Type=Bone,Name="spine_05"),Weight=0.6653))
    Weight=1
   )
   > &RigUnit_ParentConstraint(
    AdvancedSettings=(RotationOrderForFilter=XZY)
    Child=(Type=Bone,Name="dyn_pauldron_top_01_r_MSTR")
    Filter=(TranslationFilter=(bX=True,bY=True,bZ=True),RotationFilter=(bX=false,bY=false,bZ=false),ScaleFilter=(bX=True,bY=True,bZ=True))
    Parents=((Item=(Type=Bone,Name="spine_04"),Weight=0.0369),(Item=(Type=Bone,Name="neck_01"),Weight=0.0637),(Item=(Type=Bone,Name="clavicle_r"),Weight=0.1971),(Item=(Type=Bone,Name="spine_05"),Weight=0.7023))
    Weight=1
   )
   > &RigUnit_ParentConstraint(
    AdvancedSettings=(RotationOrderForFilter=XZY)
    Child=(Type=Bone,Name="dyn_pauldron_bk_01_r_MSTR")
    Filter=(TranslationFilter=(bX=True,bY=True,bZ=True),RotationFilter=(bX=false,bY=false,bZ=false),ScaleFilter=(bX=True,bY=True,bZ=True))
    Parents=((Item=(Type=Bone,Name="clavicle_r"),Weight=0.1315),(Item=(Type=Bone,Name="neck_01"),Weight=0.1546),(Item=(Type=Bone,Name="spine_05"),Weight=0.7139))
    Weight=1
   )
   n1 = RigVMVariableNode() {variable=ParentConstraint_Weight}
   &RigUnit_ParentConstraint(
    AdvancedSettings=(RotationOrderForFilter=XZY)
    Child=(Type=Control,Name="spline_01_ctrl")
    ExecutePin=()
    Parents="((Item=(Type=Bone,Name=\"dyn_pauldron_top_02_r\"),Weight=0.0188),(Item=(Type=Bone,Name=\"dyn_pauldron_fr_02_r\"),Weight=0.0194),(Item=(Type=Bone,Name=\"dyn_pauldron_fr_04_r\"),Weight=0.03),(Item=(Type=Bone,Name=\"dyn_pauldron_top_03_r\"),Weight=0.0311),(Item=(Type ...(truncated)"
    Weight=@n1.Value
   )
   > &RigUnit_ParentConstraint(
    AdvancedSettings=(RotationOrderForFilter=XZY)
    Child=(Type=Control,Name="spline_02_ctrl")
    Parents="((Item=(Type=Bone,Name=\"dyn_pauldron_fr_02_r\"),Weight=0.0219),(Item=(Type=Bone,Name=\"dyn_pauldron_fr_05_r\"),Weight=0.0226),(Item=(Type=Bone,Name=\"dyn_pauldron_fr_03_r\"),Weight=0.0251),(Item=(Type=Bone,Name=\"dyn_pauldron_top_02_r\"),Weight=0.0311),(Item=(Typ ...(truncated)"
    Weight=@n1.Value
   )
   > &RigUnit_ParentConstraint(
    AdvancedSettings=(RotationOrderForFilter=XZY)
    Child=(Type=Control,Name="spline_03_ctrl")
    Parents="((Item=(Type=Bone,Name=\"dyn_pauldron_top_02_r\"),Weight=0.0198),(Item=(Type=Bone,Name=\"dyn_pauldron_fr_02_r\"),Weight=0.0223),(Item=(Type=Bone,Name=\"dyn_pauldron_fr_01_r\"),Weight=0.033),(Item=(Type=Bone,Name=\"dyn_pauldron_fr_06_r\"),Weight=0.0673),(Item=(Type ...(truncated)"
    Weight=@n1.Value
   )
   > &RigUnit_ParentConstraint(
    AdvancedSettings=(RotationOrderForFilter=XZY)
    Child=(Type=Control,Name="spline_04_ctrl")
    Parents="((Item=(Type=Bone,Name=\"dyn_pauldron_top_02_r\"),Weight=0.0117),(Item=(Type=Bone,Name=\"dyn_pauldron_fr_02_r\"),Weight=0.0174),(Item=(Type=Bone,Name=\"dyn_pauldron_fr_06_r\"),Weight=0.0284),(Item=(Type=Bone,Name=\"dyn_pauldron_fr_01_r\"),Weight=0.0479),(Item=(Typ ...(truncated)"
    Weight=@n1.Value
   )
   > n2 = &RigUnit_SphericalPoseReader(
    ActiveRegionScaleFactors=(PositiveWidth=0.2,NegativeWidth=0.1,PositiveHeight=0.5,NegativeHeight=0.2)
    ActiveRegionSize=0.6
    Debug=(bDrawDebug=false,bDraw2D=False,bDrawLocalAxes=False,DebugScale=25,DebugSegments=20,DebugThickness=0.25)
    DriverAxis=(X=-1,Y=0,Z=0)
    DriverItem=(Type=Bone,Name="upperarm_r")
    FalloffRegionScaleFactors=(PositiveWidth=1,NegativeWidth=0.1,PositiveHeight=1,NegativeHeight=1)
    FalloffSize=0.6
    FlipHeightScaling=false
    FlipWidthScaling=false
    OptionalParentItem=(Type=Bone,Name="None")
    RotationOffset=(X=-85.093,Y=29.1525,Z=7.3673)
   )
   n3 = &RigUnit_GetTransform(Item=(Type=Control,Name="spline_01_ctrl"), bInitial=False)
   n4 = &RigUnit_GetTransform(Item=(Type=Control,Name="spline_02_ctrl"), bInitial=False)
   n5 = &RigUnit_GetTransform(Item=(Type=Control,Name="spline_03_ctrl"), bInitial=False)
   n6 = &RigUnit_GetTransform(Item=(Type=Control,Name="spline_04_ctrl"), bInitial=False)
   n7 = /Script/ControlRigSpline.RigUnit_ControlRigSplineFromTransforms(
    Compression=0
    SamplesPerSegment=10
    SplineMode=Hermite
    Stretch=0
    Transforms="((Rotation=(X=0,Y=0,Z=0,W=1),Translation=(X=0,Y=0,Z=0),Scale3D=(X=1,Y=1,Z=1)),(Rotation=(X=0,Y=0,Z=0,W=1),Translation=(X=0,Y=0,Z=0),Scale3D=(X=1,Y=1,Z=1)),(Rotation=(X=0,Y=0,Z=0,W=1),Translation=(X=0,Y=0,Z=0),Scale3D=(X=1,Y=1,Z=1)),(Rotation=(X=0,Y=0,Z=0,W ...(truncated)"
    bClosed=false
    Transforms.0=@n3.Transform
    Transforms.1=@n4.Transform
    Transforms.2=@n5.Transform
    Transforms.3=@n6.Transform
   ) {is=struct RigVMUnitNode}
   n8 = /Script/ControlRigSpline.RigUnit_TransformFromControlRigSpline2(PrimaryAxis=(X=0,Y=-1,Z=0), SecondaryAxis=(X=0,Y=0,Z=0), Spline=@n7.Spline, U=@n2.OutputParam) {is=struct RigVMUnitNode}
   > RigUnit_SetTransform = &RigUnit_SetTransform(Item=(Type=Bone,Name="dyn_pauldron_fr_strap_pin_r_MSTR"), Weight=1, bInitial=false, bPropagateToChildren=true, Value=@n8.Transform)
   RigUnit_BeginExecution = &RigUnit_BeginExecution()
   RigVMFunction_Sequence = RigVMAggregateNode(ExecuteContext=@RigUnit_BeginExecution.ExecutePin)
   n9 = &RigUnit_SphericalPoseReader(
    ActiveRegionScaleFactors=(PositiveWidth=0.1,NegativeWidth=0.1,PositiveHeight=0.1,NegativeHeight=0.1)
    ActiveRegionSize=0.4
    Debug=(bDrawDebug=false,bDraw2D=False,bDrawLocalAxes=False,DebugScale=25,DebugSegments=20,DebugThickness=0.25)
    DriverAxis=(X=1,Y=0,Z=0)
    DriverItem=(Type=Bone,Name="upperarm_l")
    FalloffRegionScaleFactors=(PositiveWidth=1,NegativeWidth=1,PositiveHeight=1,NegativeHeight=1)
    FalloffSize=0.5
    FlipHeightScaling=False
    FlipWidthScaling=False
    OptionalParentItem=(Type=Bone,Name="spine_05")
    RotationOffset=(X=112.4728,Y=-53.8018,Z=-25.0742)
    ExecutePin=@RigVMFunction_Sequence.D
   )
   > &RigUnit_ModifyTransforms(
    ItemToModify=((Item=(Type=Bone,Name="dyn_chest_rig_l"),Transform=(Rotation=(X=-0.1548,Y=0.0342,Z=0.0036,W=0.9873),Translation=(X=0.7265,Y=4.9851,Z=0.4619),Scale3D=(X=1,Y=1,Z=1))))
    WeightMinimum=0
    Weight=@n9.OutputParam
   )
   > n10 = &RigUnit_SphericalPoseReader(
    ActiveRegionScaleFactors=(PositiveWidth=0.1,NegativeWidth=0.1,PositiveHeight=0.1,NegativeHeight=0.1)
    ActiveRegionSize=0.4
    Debug=(bDrawDebug=false,bDraw2D=False,bDrawLocalAxes=False,DebugScale=25,DebugSegments=20,DebugThickness=0.25)
    DriverAxis=(X=1,Y=0,Z=0)
    DriverItem=(Type=Bone,Name="thigh_r")
    FalloffRegionScaleFactors=(PositiveWidth=1,NegativeWidth=1,PositiveHeight=1,NegativeHeight=1)
    FalloffSize=0.5
    FlipHeightScaling=False
    FlipWidthScaling=False
    OptionalParentItem=(Type=Bone,Name="None")
    RotationOffset=(X=-60.4542,Y=7.5145,Z=0)
   )
   > &RigUnit_ModifyTransforms(
    ItemToModify=((Item=(Type=Bone,Name="dyn_se14_holster_base_MSTR"),Transform=(Rotation=(X=-0.1294,Y=-0.1294,Z=-0.017,W=0.983),Translation=(X=-1,Y=0,Z=2),Scale3D=(X=1,Y=1,Z=1))))
    WeightMinimum=0
    Weight=@n10.OutputParam
   )
   &RigUnit_ParentConstraint(
    AdvancedSettings=(RotationOrderForFilter=XZY)
    Child=(Type=Bone,Name="dyn_se14_holster_base_MSTR")
    Parents="((Item=(Type=Bone,Name=\"spine_01\"),Weight=0.0165),(Item=(Type=Bone,Name=\"spine_02\"),Weight=0.0212),(Item=(Type=Bone,Name=\"deform_glute_r\"),Weight=0.0235),(Item=(Type=Bone,Name=\"deform_groin_r\"),Weight=0.1608),(Item=(Type=Bone,Name=\"thigh_twist_01_r\"),Weigh ...(truncated)"
    Weight=1
    ExecutePin=@RigVMFunction_Sequence.C
   )
   > &RigUnit_ParentConstraint(
    AdvancedSettings=(RotationOrderForFilter=XZY)
    Child=(Type=Bone,Name="dyn_back_canister_MSTR")
    Parents="((Item=(Type=Bone,Name=\"thigh_twist_01_l\"),Weight=0.015),(Item=(Type=Bone,Name=\"deform_glute_l\"),Weight=0.0171),(Item=(Type=Bone,Name=\"spine_04\"),Weight=0.0175),(Item=(Type=Bone,Name=\"spine_03\"),Weight=0.1038),(Item=(Type=Bone,Name=\"spine_01\"),Weight=0.108 ...(truncated)"
    Weight=1
   )
   > &RigUnit_ParentConstraint(
    AdvancedSettings=(RotationOrderForFilter=XZY)
    Child=(Type=Bone,Name="dyn_chest_rig_l_MSTR")
    Parents=((Item=(Type=Bone,Name="spine_03"),Weight=0.0758),(Item=(Type=Bone,Name="spine_05"),Weight=0.2166),(Item=(Type=Bone,Name="spine_04"),Weight=0.2521),(Item=(Type=Bone,Name="clavicle_l"),Weight=0.4555))
    Weight=1
   )
   > &RigUnit_ParentConstraint(
    AdvancedSettings=(RotationOrderForFilter=XZY)
    Child=(Type=Bone,Name="dyn_belt_pouch_01_MSTR")
    Parents="((Item=(Type=Bone,Name=\"thigh_twist_01_r\"),Weight=0.0149),(Item=(Type=Bone,Name=\"deform_groin_r\"),Weight=0.0241),(Item=(Type=Bone,Name=\"spine_03\"),Weight=0.0396),(Item=(Type=Bone,Name=\"spine_01\"),Weight=0.0508),(Item=(Type=Bone,Name=\"spine_02\"),Weight=0.06 ...(truncated)"
    Weight=1
   )
   > &RigUnit_ParentConstraint(
    AdvancedSettings=(RotationOrderForFilter=XZY)
    Child=(Type=Bone,Name="dyn_belt_pouch_02_MSTR")
    Parents="((Item=(Type=Bone,Name=\"deform_groin_r\"),Weight=0.0128),(Item=(Type=Bone,Name=\"deform_glute_l\"),Weight=0.0139),(Item=(Type=Bone,Name=\"spine_03\"),Weight=0.0184),(Item=(Type=Bone,Name=\"spine_01\"),Weight=0.0263),(Item=(Type=Bone,Name=\"spine_02\"),Weight=0.0327 ...(truncated)"
    Weight=1
   )
   > &RigUnit_ParentConstraint(
    AdvancedSettings=(RotationOrderForFilter=XZY)
    Child=(Type=Bone,Name="dyn_belt_pouch_03_MSTR")
    Parents="((Item=(Type=Bone,Name=\"deform_groin_l\"),Weight=0.0159),(Item=(Type=Bone,Name=\"spine_03\"),Weight=0.0187),(Item=(Type=Bone,Name=\"spine_01\"),Weight=0.0296),(Item=(Type=Bone,Name=\"spine_02\"),Weight=0.0342),(Item=(Type=Bone,Name=\"deform_glute_l\"),Weight=0.1048 ...(truncated)"
    Weight=1
   )
   > &RigUnit_ParentConstraint(
    AdvancedSettings=(RotationOrderForFilter=XZY)
    Child=(Type=Bone,Name="dyn_belt_pouch_04_MSTR")
    Parents="((Item=(Type=Bone,Name=\"deform_groin_r\"),Weight=0.0165),(Item=(Type=Bone,Name=\"spine_03\"),Weight=0.0185),(Item=(Type=Bone,Name=\"spine_01\"),Weight=0.0295),(Item=(Type=Bone,Name=\"spine_02\"),Weight=0.034),(Item=(Type=Bone,Name=\"deform_glute_r\"),Weight=0.1042) ...(truncated)"
    Weight=1
   )
   > &RigUnit_ParentConstraint(
    AdvancedSettings=(RotationOrderForFilter=XZY)
    Child=(Type=Bone,Name="dyn_belt_pouch_05_MSTR")
    Parents="((Item=(Type=Bone,Name=\"thigh_twist_01_l\"),Weight=0.0118),(Item=(Type=Bone,Name=\"deform_groin_l\"),Weight=0.0199),(Item=(Type=Bone,Name=\"spine_03\"),Weight=0.044),(Item=(Type=Bone,Name=\"spine_01\"),Weight=0.0562),(Item=(Type=Bone,Name=\"spine_02\"),Weight=0.070 ...(truncated)"
    Weight=1
   )
   > &RigUnit_ParentConstraint(
    AdvancedSettings=(RotationOrderForFilter=XZY)
    Child=(Type=Bone,Name="dyn_belt_pouch_06_MSTR")
    Parents="((Item=(Type=Bone,Name=\"spine_01\"),Weight=0.0197),(Item=(Type=Bone,Name=\"spine_02\"),Weight=0.0229),(Item=(Type=Bone,Name=\"deform_groin_r\"),Weight=0.0477),(Item=(Type=Bone,Name=\"deform_glute_r\"),Weight=0.077),(Item=(Type=Bone,Name=\"thigh_twist_01_r\"),Weight ...(truncated)"
    Weight=1
   )
   > &RigUnit_ParentConstraint(
    AdvancedSettings=(RotationOrderForFilter=XZY)
    Child=(Type=Bone,Name="dyn_belly_rig_pouch_l_01_MSTR")
    Parents=((Item=(Type=Bone,Name="spine_04"),Weight=0.038),(Item=(Type=Bone,Name="spine_01"),Weight=0.1629),(Item=(Type=Bone,Name="spine_03"),Weight=0.1752),(Item=(Type=Bone,Name="spine_02"),Weight=0.2309),(Item=(Type=Bone,Name="pelvis"),Weight=0.3931))
    Weight=1
   )
   > &RigUnit_ParentConstraint(
    AdvancedSettings=(RotationOrderForFilter=XZY)
    Child=(Type=Bone,Name="dyn_belly_rig_pouch_l_02_MSTR")
    Parents=((Item=(Type=Bone,Name="spine_04"),Weight=0.0382),(Item=(Type=Bone,Name="spine_01"),Weight=0.1664),(Item=(Type=Bone,Name="spine_03"),Weight=0.181),(Item=(Type=Bone,Name="spine_02"),Weight=0.2365),(Item=(Type=Bone,Name="pelvis"),Weight=0.3779))
    Weight=1
   )
   > &RigUnit_ParentConstraint(
    AdvancedSettings=(RotationOrderForFilter=XZY)
    Child=(Type=Bone,Name="dyn_belly_rig_pouch_l_03")
    Parents=((Item=(Type=Bone,Name="spine_04"),Weight=0.0375),(Item=(Type=Bone,Name="spine_01"),Weight=0.1657),(Item=(Type=Bone,Name="spine_03"),Weight=0.1778),(Item=(Type=Bone,Name="spine_02"),Weight=0.2337),(Item=(Type=Bone,Name="pelvis"),Weight=0.3853))
    Weight=1
   )
   > &RigUnit_ParentConstraint(
    AdvancedSettings=(RotationOrderForFilter=XZY)
    Child=(Type=Bone,Name="dyn_belly_rig_pouch_r_01_MSTR")
    Parents=((Item=(Type=Bone,Name="spine_04"),Weight=0.0378),(Item=(Type=Bone,Name="spine_01"),Weight=0.1627),(Item=(Type=Bone,Name="spine_03"),Weight=0.1748),(Item=(Type=Bone,Name="spine_02"),Weight=0.2304),(Item=(Type=Bone,Name="pelvis"),Weight=0.3943))
    Weight=1
   )
   > &RigUnit_ParentConstraint(
    AdvancedSettings=(RotationOrderForFilter=XZY)
    Child=(Type=Bone,Name="dyn_belly_rig_pouch_r_02_MSTR")
    Parents=((Item=(Type=Bone,Name="spine_04"),Weight=0.0375),(Item=(Type=Bone,Name="spine_01"),Weight=0.1653),(Item=(Type=Bone,Name="spine_03"),Weight=0.1793),(Item=(Type=Bone,Name="spine_02"),Weight=0.2347),(Item=(Type=Bone,Name="pelvis"),Weight=0.3832))
    Weight=1
   )
   > &RigUnit_ParentConstraint(
    AdvancedSettings=(RotationOrderForFilter=XZY)
    Child=(Type=Bone,Name="dyn_belly_rig_pouch_r_03")
    Parents=((Item=(Type=Bone,Name="spine_04"),Weight=0.0375),(Item=(Type=Bone,Name="spine_01"),Weight=0.1657),(Item=(Type=Bone,Name="spine_03"),Weight=0.1778),(Item=(Type=Bone,Name="spine_02"),Weight=0.2337),(Item=(Type=Bone,Name="pelvis"),Weight=0.3853))
    Weight=1
   )
   &RigUnit_ParentConstraint(
    AdvancedSettings=(RotationOrderForFilter=XZY)
    Child=(Type=Bone,Name="dyn_pauldron_bk_bott_strap_pin_r_MSTR")
    Parents="((Item=(Type=Bone,Name=\"spine_02\"),Weight=0.1),(Item=(Type=Bone,Name=\"spine_03\"),Weight=0.1),(Item=(Type=Bone,Name=\"clavicle_r\"),Weight=0.3),(Item=(Type=Bone,Name=\"spine_04\"),Weight=0.2),(Item=(Type=Bone,Name=\"upperarm_twist_02_r\"),Weight=0.1253),(Item=(Ty ...(truncated)"
    Weight=1
    ExecutePin=@RigVMFunction_Sequence.K
   )
   > &RigUnit_ParentConstraint(
    AdvancedSettings=(RotationOrderForFilter=XZY)
    Child=(Type=Bone,Name="dyn_pauldron_fr_bott_strap_pin_r_MSTR")
    Parents="((Item=(Type=Bone,Name=\"spine_02\"),Weight=0.05),(Item=(Type=Bone,Name=\"upperarm_twist_02_r\"),Weight=0.1),(Item=(Type=Bone,Name=\"upperarm_r\"),Weight=0.0631),(Item=(Type=Bone,Name=\"deform_upperarm_fr_r\"),Weight=0.3),(Item=(Type=Bone,Name=\"clavicle_r\"),Weight ...(truncated)"
    Weight=1
   )
   > &RigUnit_ParentConstraint(
    AdvancedSettings=(RotationOrderForFilter=XZY)
    Child=(Type=Bone,Name="dyn_pauldron_bk_strap_pin_r_MSTR")
    Parents="((Item=(Type=Bone,Name=\"dyn_pauldron_top_02_r\"),Weight=0.0539),(Item=(Type=Bone,Name=\"dyn_pauldron_top_03_r\"),Weight=0.0688),(Item=(Type=Bone,Name=\"dyn_pauldron_top_04_r\"),Weight=0.0766),(Item=(Type=Bone,Name=\"clavicle_r\"),Weight=0.0842),(Item=(Type=Bone,N ...(truncated)"
    Weight=1
   )
   FN_DeformRig_FNC = RigVMFunctionReferenceNode(
    DeformRig_ Curve="(deformRig_l_upperarm_up_alpha_72_#1=False,deformRig_r_upperarm_up_alpha_73_#2=False,deformRig_l_upperarm_fr_alpha_74_#3=False,deformRig_r_upperarm_fr_alpha_75_CB51B8 ...(truncated)"
    DeformRig_Bool="(upperarm_1_#4=True,elbow_3_#5=True,wrist_7_#6=False,groin_8_#7=True,glute_17_#8=True,knee_9_40C408324CA26 ...(truncated)"
    DeformRig_Debug=(upperarm_debug_18_#4=False,wrist_debug_20_#6=False,lat_debug_26_#9=False,groinGlute_debug_27_#7=False)
    DeformRig_Value="(upperarm_up_18_#10=(Rotation=(X=0,Y=0,Z=0,W=1),Translation=(X=0,Y=0,Z=8),Scale3D=(X=0.7,Y=1,Z=1)),upperarm_fwd_fr_19_#11=(Rotation=(X=0,Y=0,Z=0,W=1),Translation=(X=0,Y=10,Z=0),Scale3D=(X=0.5,Y=1,Z= ...(truncated)"
    ExecuteContext=@RigVMFunction_Sequence.B
   )
   &RigUnit_ModifyTransforms(
    ItemToModify=((Item=(Type=Bone,Name="dyn_shoulderpad_l"),Transform=(Rotation=(X=-0.0698,Y=0,Z=0,W=0.9976),Translation=(X=0,Y=0,Z=0),Scale3D=(X=1,Y=1,Z=1))))
    Weight=1
    WeightMinimum=0
    ExecutePin=@RigVMFunction_Sequence.E
   )
   > n11 = &RigUnit_ParentConstraint(
    AdvancedSettings=(RotationOrderForFilter=XZY)
    Child=(Type=Bone,Name="dyn_shoulderpad_dist_bott_l_MSTR")
    Parents="((Item=(Type=Bone,Name=\"deform_upperarm_fr_l\"),Weight=0.0136),(Item=(Type=Bone,Name=\"spine_04\"),Weight=0.0379),(Item=(Type=Bone,Name=\"deform_upperarm_up_l\"),Weight=0.0415),(Item=(Type=Bone,Name=\"spine_05\"),Weight=0.3833),(Item=(Type=Bone,Name=\"clavicle_l\") ...(truncated)"
    Weight=1
   )
   n12 = &RigVMFunction_Sequence(ExecuteContext=@n11.ExecutePin)
   n13 = &RigUnit_SphericalPoseReader(
    ActiveRegionScaleFactors=(PositiveWidth=1,NegativeWidth=1,PositiveHeight=1,NegativeHeight=1)
    ActiveRegionSize=0.01
    Debug=(bDrawDebug=false,bDraw2D=False,bDrawLocalAxes=False,DebugScale=25,DebugSegments=20,DebugThickness=0.25)
    DriverAxis=(X=1,Y=0,Z=0)
    DriverItem=(Type=Bone,Name="upperarm_l")
    FalloffRegionScaleFactors=(PositiveWidth=1,NegativeWidth=1,PositiveHeight=1,NegativeHeight=1)
    FalloffSize=0.63
    FlipHeightScaling=False
    FlipWidthScaling=False
    OptionalParentItem=(Type=Bone,Name="None")
    RotationOffset=(X=15,Y=13,Z=0)
    ExecutePin=@n12.A
   )
   RigUnit_GetTransform = &RigUnit_GetTransform(Item=(Type=Bone,Name="dyn_shoulderpad_dist_top_l"), bInitial=False)
   RigVMFunction_MathVectorDistance = &RigVMFunction_MathVectorDistance(
    A=@RigUnit_GetTransform.Transform.Translation
    B=&RigUnit_GetTransform(Item=(Type=Bone,Name="dyn_shoulderpad_dist_bott_l"), bInitial=False).Transform.Translation
   )
   Remap = &RigVMFunction_MathDoubleRemap(SourceMaximum=5.8, SourceMinimum=1.3, TargetMaximum=1, TargetMinimum=0, Value=@RigVMFunction_MathVectorDistance.Result)
   Multiply = &RigVMFunction_MathDoubleMul(B=@Remap.Result, A=@n13.OutputParam)
   > &RigUnit_ModifyTransforms(
    ItemToModify=((Item=(Type=Bone,Name="dyn_shoulderpad_l"),Transform=(Rotation=(X=0,Y=0.2588,Z=0,W=0.9659),Translation=(X=12,Y=0,Z=4.5),Scale3D=(X=1,Y=1,Z=1))))
    WeightMinimum=0
    Weight=@Multiply.Result
   )
   > &RigUnit_ModifyTransforms(
    ItemToModify=((Item=(Type=Bone,Name="dyn_shoulderpad_l"),Transform=(Rotation=(X=0,Y=-0.1564,Z=0,W=0.9877),Translation=(X=2.7794,Y=0,Z=-0.6772),Scale3D=(X=1,Y=1,Z=1))))
    WeightMinimum=0
    Weight=&RigVMFunction_MathDoubleMul(
     B=&RigVMFunction_MathDoubleRemap(SourceMaximum=13.2, SourceMinimum=5.8, TargetMaximum=1, TargetMinimum=0, Value=@RigVMFunction_MathVectorDistance.Result).Result
     A=@n13.OutputParam
    ).Result
   )
   n14 = &RigUnit_SphericalPoseReader(
    ActiveRegionScaleFactors=(PositiveWidth=1,NegativeWidth=1,PositiveHeight=1,NegativeHeight=1)
    ActiveRegionSize=0.01
    Debug=(bDrawDebug=false,bDraw2D=False,bDrawLocalAxes=False,DebugScale=25,DebugSegments=20,DebugThickness=0.25)
    DriverAxis=(X=1,Y=0,Z=0)
    DriverItem=(Type=Bone,Name="upperarm_l")
    FalloffRegionScaleFactors=(PositiveWidth=1,NegativeWidth=1,PositiveHeight=1,NegativeHeight=1)
    FalloffSize=0.3
    FlipHeightScaling=False
    FlipWidthScaling=False
    OptionalParentItem=(Type=Bone,Name="None")
    RotationOffset=(X=15,Y=-130,Z=0)
    ExecutePin=@n12.B
   )
   > &RigUnit_ModifyTransforms(
    ItemToModify=((Item=(Type=Bone,Name="dyn_shoulderpad_l"),Transform=(Rotation=(X=-0.1027,Y=-0.0224,Z=0.0016,W=0.9945),Translation=(X=5.4913,Y=0.1813,Z=0.9777),Scale3D=(X=1,Y=1,Z=1))))
    WeightMinimum=0
    Weight=&RigVMFunction_MathDoubleMul(
     B=&RigVMFunction_MathDoubleRemap(SourceMaximum=7.8, SourceMinimum=1.3, TargetMaximum=1, TargetMinimum=0, Value=@RigVMFunction_MathVectorDistance.Result).Result
     A=@n14.OutputParam
    ).Result
   )
   n15 = &RigUnit_SphericalPoseReader(
    ActiveRegionScaleFactors=(PositiveWidth=0,NegativeWidth=0,PositiveHeight=0,NegativeHeight=0)
    ActiveRegionSize=0.4
    Debug=(bDrawDebug=false,bDraw2D=False,bDrawLocalAxes=False,DebugScale=25,DebugSegments=20,DebugThickness=0.25)
    DriverAxis=(X=-1,Y=0,Z=0)
    DriverItem=(Type=Bone,Name="upperarm_r")
    FalloffRegionScaleFactors=(PositiveWidth=0.8,NegativeWidth=0.8,PositiveHeight=0.8,NegativeHeight=0.8)
    FalloffSize=0.5
    FlipHeightScaling=false
    FlipWidthScaling=false
    OptionalParentItem=(Type=Bone,Name="None")
    RotationOffset=(X=-122.2118,Y=23.6646,Z=7.6678)
    ExecutePin=@RigVMFunction_Sequence.F
   )
   > &RigUnit_ModifyTransforms(
    ItemToModify=((Item=(Type=Bone,Name="dyn_pauldron_fr_01_r_MSTR"),Transform=(Rotation=(X=-0.687,Y=0,Z=0,W=0.7267),Translation=(X=-0.0032,Y=2.2528,Z=3.7653),Scale3D=(X=1,Y=1,Z=1))))
    WeightMinimum=0
    Weight=@n15.OutputParam
   )
   > &RigUnit_ModifyTransforms(
    ItemToModify=((Item=(Type=Bone,Name="dyn_pauldron_fr_07_r_MSTR"),Transform=(Rotation=(X=0.3437,Y=0.0965,Z=-0.0355,W=0.9335),Translation=(X=-1.2317,Y=0.4094,Z=0.8833),Scale3D=(X=1,Y=1,Z=1))))
    WeightMinimum=0
    Weight=@n15.OutputParam
   )
   > &RigUnit_ModifyTransforms(
    ItemToModify=((Item=(Type=Bone,Name="dyn_pauldron_fr_05_r"),Transform=(Rotation=(X=0.4396,Y=0.2111,Z=-0.3895,W=0.7813),Translation=(X=-1.0383,Y=-1.6072,Z=-6.6214),Scale3D=(X=1,Y=1,Z=1))))
    WeightMinimum=0
    Weight=@n15.OutputParam
   )
   > &RigUnit_ModifyTransforms(
    ItemToModify=((Item=(Type=Bone,Name="dyn_pauldron_top_02_r_MSTR"),Transform=(Rotation=(X=-0.3665,Y=0,Z=0,W=0.9304),Translation=(X=0,Y=-0.1318,Z=0.6552),Scale3D=(X=1,Y=1,Z=1))))
    WeightMinimum=0
    Weight=@n15.OutputParam
   )
   > n16 = &RigUnit_SphericalPoseReader(
    ActiveRegionScaleFactors=(PositiveWidth=0.2,NegativeWidth=0.1,PositiveHeight=0.5,NegativeHeight=0.2)
    ActiveRegionSize=0.6
    Debug=(bDrawDebug=false,bDraw2D=False,bDrawLocalAxes=False,DebugScale=25,DebugSegments=20,DebugThickness=0.25)
    DriverAxis=(X=-1,Y=0,Z=0)
    DriverItem=(Type=Bone,Name="upperarm_r")
    FalloffRegionScaleFactors=(PositiveWidth=1,NegativeWidth=0.1,PositiveHeight=1,NegativeHeight=1)
    FalloffSize=0.6
    FlipHeightScaling=false
    FlipWidthScaling=false
    OptionalParentItem=(Type=Bone,Name="None")
    RotationOffset=(X=-85.093,Y=29.1525,Z=7.3673)
   )
   > &RigUnit_ModifyTransforms(
    ItemToModify=((Item=(Type=Bone,Name="dyn_pauldron_fr_strap_pin_r"),Transform=(Rotation=(X=-0.2032,Y=-0.0652,Z=-0.499,W=0.8399),Translation=(X=-0.4292,Y=-1.0733,Z=0.7585),Scale3D=(X=1,Y=1,Z=1))))
    WeightMinimum=0
    Weight=@n16.OutputParam
   )
   n17 = &RigUnit_SphericalPoseReader(
    ActiveRegionScaleFactors=(PositiveWidth=0.8,NegativeWidth=0.05,PositiveHeight=0.3,NegativeHeight=0.3)
    ActiveRegionSize=0.4
    Debug=(bDrawDebug=false,bDraw2D=False,bDrawLocalAxes=False,DebugScale=25,DebugSegments=20,DebugThickness=0.25)
    DriverAxis=(X=-1,Y=0,Z=0)
    DriverItem=(Type=Bone,Name="upperarm_r")
    FalloffRegionScaleFactors=(PositiveWidth=1,NegativeWidth=1,PositiveHeight=0.5,NegativeHeight=0.5)
    FalloffSize=0.5
    FlipHeightScaling=false
    FlipWidthScaling=false
    OptionalParentItem=(Type=Bone,Name="spine_05")
    RotationOffset=(X=175.8156,Y=-11.2124,Z=42.3409)
    ExecutePin=@RigVMFunction_Sequence.G
   )
   > &RigUnit_ModifyTransforms(
    ItemToModify=((Item=(Type=Bone,Name="dyn_pauldron_fr_02_r_MSTR"),Transform=(Rotation=(X=-0.2034,Y=-0.2379,Z=0.0044,W=0.9497),Translation=(X=0,Y=0,Z=0),Scale3D=(X=1,Y=1,Z=1))))
    WeightMinimum=0
    Weight=@n17.OutputParam
   )
   > &RigUnit_ModifyTransforms(
    ItemToModify=((Item=(Type=Bone,Name="dyn_pauldron_top_02_r_MSTR"),Transform=(Rotation=(X=-0.0426,Y=-0.2129,Z=-0.029,W=0.9757),Translation=(X=0,Y=0,Z=0),Scale3D=(X=1,Y=1,Z=1))))
    WeightMinimum=0
    Weight=@n17.OutputParam
   )
   n18 = &RigUnit_ItemArray(Items="((Type=Bone,Name=\"dyn_pauldron_fr_03_r_MSTR\"),(Type=Bone,Name=\"dyn_pauldron_top_03_r_MSTR\"),(Type=Bone,Name=\"dyn_pauldron_topbk_02_r_MSTR\"),(Type=Bone,Name=\"dyn_pauldron_fr_02_r_MSTR\"),(Type=Bone,Name=\"dyn_pauldron_fr_06_r_MSTR\"),(Type=Bone,Name=\"dyn_pauld ...(truncated)")
   DISPATCH_RigVMDispatch_ArrayIterator = RigVMDispatchNode(Array=@n18.Items, ExecuteContext=@RigVMFunction_Sequence.H) {template="DISPATCH_RigVMDispatch_ArrayIterator(in Array,out Element,out Index,out Count,out Ratio)"}
   ClampSpatially = /Script/RigVM.RigVMFunction_MathTransformClampSpatially(
    Axis=Y
    DebugColor=(R=1,G=0,B=0,A=1)
    DebugThickness=1
    Maximum=40
    Minimum=7
    Type=Capsule
    bDrawDebug=false
    Space=&RigUnit_GetTransform(Item=(Type=Control,Name="upperarm_r_ctrl"), bInitial=False).Transform
    Value=&RigUnit_GetTransform(bInitial=False, Item=@DISPATCH_RigVMDispatch_ArrayIterator.Element).Transform
   ) {
    is=struct RigVMUnitNode
    template="ClampSpatially::Execute(in Value,in Axis,in Type,in Minimum,in Maximum,in Space,in bDrawDebug,in DebugColor,in DebugThickness,out Result)"
   }
   > &RigUnit_SetTransform(Weight=1, bInitial=False, Value=@ClampSpatially.Result, Item=@DISPATCH_RigVMDispatch_ArrayIterator.Element)
   n19 = &RigUnit_SphericalPoseReader(
    ActiveRegionScaleFactors=(PositiveWidth=0.8,NegativeWidth=0.05,PositiveHeight=0.3,NegativeHeight=0.3)
    ActiveRegionSize=0.4
    Debug=(bDrawDebug=false,bDraw2D=False,bDrawLocalAxes=False,DebugScale=25,DebugSegments=20,DebugThickness=0.25)
    DriverAxis=(X=1,Y=0,Z=0)
    DriverItem=(Type=Bone,Name="thigh_r")
    FalloffRegionScaleFactors=(PositiveWidth=1,NegativeWidth=1,PositiveHeight=0.5,NegativeHeight=0.5)
    FalloffSize=0.5
    FlipHeightScaling=false
    FlipWidthScaling=true
    OptionalParentItem=(Type=Bone,Name="None")
    RotationOffset=(X=-58.2527,Y=-6.7165,Z=-22.8064)
    ExecutePin=@RigVMFunction_Sequence.J
   )
   > &RigUnit_ModifyTransforms(
    ItemToModify=((Item=(Type=Bone,Name="dyn_belt_pouch_05_MSTR"),Transform=(Rotation=(X=0.2826,Y=-0.7249,Z=-0.0533,W=0.626),Translation=(X=1.3755,Y=-0.1547,Z=0.0298),Scale3D=(X=1,Y=1,Z=1))))
    WeightMinimum=0
    Weight=@n19.OutputParam
   )
   > &RigUnit_ModifyTransforms(
    ItemToModify=((Item=(Type=Bone,Name="dyn_belly_rig_pouch_r_02_MSTR"),Transform=(Rotation=(X=0,Y=0,Z=0,W=1),Translation=(X=-8,Y=0,Z=0),Scale3D=(X=1,Y=1,Z=1))))
    WeightMinimum=0
    Weight=@n19.OutputParam
   )
   > &RigUnit_ModifyTransforms(
    ItemToModify=((Item=(Type=Bone,Name="dyn_belly_rig_pouch_r_03"),Transform=(Rotation=(X=0,Y=0,Z=0,W=1),Translation=(X=-5,Y=0,Z=0),Scale3D=(X=1,Y=1,Z=1))))
    WeightMinimum=0
    Weight=@n19.OutputParam
   )
   > &RigUnit_ModifyTransforms(
    ItemToModify=((Item=(Type=Bone,Name="dyn_belly_rig_pouch_r_01_MSTR"),Transform=(Rotation=(X=0,Y=0,Z=0,W=1),Translation=(X=-1,Y=0,Z=0),Scale3D=(X=1,Y=1,Z=1))))
    WeightMinimum=0
    Weight=@n19.OutputParam
   )
   > &RigUnit_ModifyTransforms(
    ItemToModify=((Item=(Type=Bone,Name="dyn_se14_holster_base_MSTR"),Transform=(Rotation=(X=-0.1282,Y=-0.0141,Z=0.0348,W=0.991),Translation=(X=0,Y=0,Z=0),Scale3D=(X=1,Y=1,Z=1))))
    Weight=1
    WeightMinimum=0
   )
   n20 = &RigUnit_SphericalPoseReader(
    ActiveRegionScaleFactors=(PositiveWidth=0.8,NegativeWidth=0.05,PositiveHeight=0.3,NegativeHeight=0.3)
    ActiveRegionSize=0.4
    Debug=(bDrawDebug=false,bDraw2D=False,bDrawLocalAxes=False,DebugScale=25,DebugSegments=20,DebugThickness=0.25)
    DriverAxis=(X=-1,Y=0,Z=0)
    DriverItem=(Type=Bone,Name="thigh_l")
    FalloffRegionScaleFactors=(PositiveWidth=1,NegativeWidth=1,PositiveHeight=0.5,NegativeHeight=0.5)
    FalloffSize=0.5
    FlipHeightScaling=false
    FlipWidthScaling=false
    OptionalParentItem=(Type=Bone,Name="None")
    RotationOffset=(X=96.7681,Y=-1.5612,Z=-42.9257)
    ExecutePin=@RigVMFunction_Sequence.I
   )
   > &RigUnit_ModifyTransforms(
    ItemToModify=((Item=(Type=Bone,Name="dyn_belt_pouch_01_MSTR"),Transform=(Rotation=(X=-0.4151,Y=-0.867,Z=-0.2489,W=0.1187),Translation=(X=8.5999,Y=2.0037,Z=2.1442),Scale3D=(X=1,Y=1,Z=1))))
    WeightMinimum=0
    Weight=@n20.OutputParam
   )
   > &RigUnit_ModifyTransforms(
    ItemToModify=((Item=(Type=Bone,Name="dyn_belt_pouch_02_MSTR"),Transform=(Rotation=(X=0,Y=0,Z=0,W=1),Translation=(X=3,Y=6,Z=0),Scale3D=(X=1,Y=1,Z=1))))
    WeightMinimum=0
    Weight=@n20.OutputParam
   )
   > &RigUnit_ModifyTransforms(
    ItemToModify=((Item=(Type=Bone,Name="dyn_belly_rig_pouch_l_02_MSTR"),Transform=(Rotation=(X=0,Y=0,Z=0,W=1),Translation=(X=-11,Y=0,Z=0),Scale3D=(X=1,Y=1,Z=1))))
    WeightMinimum=0
    Weight=@n20.OutputParam
   )
   > &RigUnit_ModifyTransforms(
    ItemToModify=((Item=(Type=Bone,Name="dyn_belly_rig_pouch_l_03"),Transform=(Rotation=(X=0,Y=0,Z=0,W=1),Translation=(X=-6,Y=0,Z=0),Scale3D=(X=1,Y=1,Z=1))))
    WeightMinimum=0
    Weight=@n20.OutputParam
   )
   > &RigUnit_ModifyTransforms(
    ItemToModify=((Item=(Type=Bone,Name="dyn_belly_rig_pouch_l_01_MSTR"),Transform=(Rotation=(X=0,Y=0,Z=0,W=1),Translation=(X=-3,Y=0,Z=0),Scale3D=(X=1,Y=1,Z=1))))
    WeightMinimum=0
    Weight=@n20.OutputParam
   )
   > &RigUnit_ParentConstraint(
    AdvancedSettings=(RotationOrderForFilter=XZY)
    Child=(Type=Bone,Name="dyn_grenade_m_MSTR")
    Parents=((Item=(Type=Bone,Name="dyn_grenade_r_MSTR"),Weight=0.5),(Item=(Type=Bone,Name="dyn_grenade_l_MSTR"),Weight=0.5))
    Weight=1
   )
   &RigUnit_ParentConstraint(
    AdvancedSettings=(RotationOrderForFilter=XZY)
    Child=(Type=Bone,Name="attach_cape")
    Parents=((Item=(Type=Bone,Name="clavicle_r"),Weight=0.0173),(Item=(Type=Bone,Name="spine_01"),Weight=0.0233),(Item=(Type=Bone,Name="spine_04"),Weight=0.2525),(Item=(Type=Bone,Name="spine_03"),Weight=0.3306),(Item=(Type=Bone,Name="spine_05"),Weight=0.3764))
    Weight=1
    ExecutePin=@RigVMFunction_Sequence.A
   )
   > &RigUnit_ParentConstraint(
    AdvancedSettings=(RotationOrderForFilter=XZY)
    Child=(Type=Bone,Name="attach_backpack")
    Parents=((Item=(Type=Bone,Name="clavicle_l"),Weight=0.0173),(Item=(Type=Bone,Name="spine_01"),Weight=0.0233),(Item=(Type=Bone,Name="spine_04"),Weight=0.2525),(Item=(Type=Bone,Name="spine_03"),Weight=0.3306),(Item=(Type=Bone,Name="spine_05"),Weight=0.3764))
    Weight=1
   )
   /Script/RigVM.RigVMFunction_DebugTransformMutableNoSpace(
    Color=(R=1,G=0,B=0,A=1)
    DebugDrawSettings=(DepthPriority=SDPG_Foreground,Lifetime=-1)
    ExecutePin=()
    Mode=Axes
    Scale=3
    Thickness=0.25
    WorldOffset=(Rotation=(X=0,Y=0,Z=0,W=1),Translation=(X=0,Y=0,Z=0),Scale3D=(X=1,Y=1,Z=1))
    bEnabled=true
    Transform=@n8.Transform
   ) {is=struct RigVMUnitNode}
   > &RigVMFunction_DebugTransformMutableNoSpace(
    Color=(R=1,G=0,B=0,A=1)
    DebugDrawSettings=(DepthPriority=SDPG_Foreground,Lifetime=-1)
    Mode=Axes
    Scale=5
    Thickness=0.1
    WorldOffset=(Rotation=(X=0,Y=0,Z=0,W=1),Translation=(X=0,Y=0,Z=0),Scale3D=(X=1,Y=1,Z=1))
    bEnabled=true
    Transform=&RigUnit_GetTransform(Item=(Type=Bone,Name="dyn_pauldron_fr_strap_pin_r_MSTR"), bInitial=False).Transform
   )
   &RigVMFunction_DebugTransformMutableNoSpace(
    Color=(R=1,G=0,B=0,A=1)
    DebugDrawSettings=(DepthPriority=SDPG_Foreground,Lifetime=-1)
    ExecutePin=()
    Mode=Axes
    Scale=2
    Thickness=0.1
    WorldOffset=(Rotation=(X=0,Y=0,Z=0,W=1),Translation=(X=0,Y=0,Z=0),Scale3D=(X=1,Y=1,Z=1))
    bEnabled=true
    Transform=@n3.Transform
   )
   > &RigVMFunction_DebugTransformMutableNoSpace(
    Color=(R=1,G=0,B=0,A=1)
    DebugDrawSettings=(DepthPriority=SDPG_Foreground,Lifetime=-1)
    Mode=Axes
    Scale=2
    Thickness=0.1
    WorldOffset=(Rotation=(X=0,Y=0,Z=0,W=1),Translation=(X=0,Y=0,Z=0),Scale3D=(X=1,Y=1,Z=1))
    bEnabled=true
    Transform=@n4.Transform
   )
   > &RigVMFunction_DebugTransformMutableNoSpace(
    Color=(R=1,G=0,B=0,A=1)
    DebugDrawSettings=(DepthPriority=SDPG_Foreground,Lifetime=-1)
    Mode=Axes
    Scale=2
    Thickness=0.1
    WorldOffset=(Rotation=(X=0,Y=0,Z=0,W=1),Translation=(X=0,Y=0,Z=0),Scale3D=(X=1,Y=1,Z=1))
    bEnabled=true
    Transform=@n5.Transform
   )
   > &RigVMFunction_DebugTransformMutableNoSpace(
    Color=(R=1,G=0,B=0,A=1)
    DebugDrawSettings=(DepthPriority=SDPG_Foreground,Lifetime=-1)
    Mode=Axes
    Scale=2
    Thickness=0.1
    WorldOffset=(Rotation=(X=0,Y=0,Z=0,W=1),Translation=(X=0,Y=0,Z=0),Scale3D=(X=1,Y=1,Z=1))
    bEnabled=true
    Transform=@n6.Transform
   )
   > /Script/ControlRigSpline.RigUnit_DrawControlRigSpline(Color=(R=0,G=0,B=0,A=1), Detail=16, Thickness=0.08, Spline=@n7.Spline) {is=struct RigVMUnitNode}
   RigVMRerouteNode(Value=@RigVMFunction_Sequence.L)
   RigVMFunction_DebugTransformMutableNoSpace = &RigVMFunction_DebugTransformMutableNoSpace(
    Color=(R=1,G=0,B=0,A=1)
    DebugDrawSettings=(DepthPriority=SDPG_Foreground,Lifetime=-1)
    ExecutePin=()
    Mode=Axes
    Scale=10
    Thickness=0.5
    WorldOffset=(Rotation=(X=0,Y=0,Z=0,W=1),Translation=(X=0,Y=0,Z=0),Scale3D=(X=1,Y=1,Z=1))
    bEnabled=true
    Transform=&RigUnit_GetTransform(Item=(Type=Bone,Name="upperarm_twist_01_l"), bInitial=False).Transform
   )
  ]}
]
hierarchy [
 root {loc=<0 0 0>, rot=<0 0 0 1>}
 pelvis {parent=root, loc=<0 -0.7071 93.1191>, rot=<0.0127 0.707 -0.0127 -0.707>}
 spine_01 {parent=pelvis, loc=<6.7494 -0.6388 0>, rot=<0 0 0.034 -0.9994>}
 spine_02 {parent=spine_01, loc=<7.2518 -0.035 0>, rot=<0 0 0.0238 -0.9997>}
 spine_03 {parent=spine_02, loc=<6.0104 0 0>, rot=<0 0 0.0149 -0.9999>}
 spine_04 {parent=spine_03, loc=<8.1024 0 0>, rot=<0 0 -0.0132 -0.9999>}
 spine_05 {parent=spine_04, loc=<10.0146 0 0>, rot=<0 0 -0.0015 -1>}
 attach_cape {parent=spine_05, loc=<9.017 -11.714 0>, rot=<0.0616 -0.7044 -0.0616 -0.7044>}
 neck_01 {parent=spine_05, loc=<11.3138 0.1529 0>, rot=<0 0 -0.2003 -0.9797>}
 neck_02 {parent=neck_01, loc=<4.5627 0 0>, rot=<0 0 0.0504 -0.9987>}
 head {parent=neck_02, loc=<5.3914 0 0>, rot=<0 0 0.115 -0.9934>}
 attach_backpack {parent=spine_05, loc=<-0.123 -13.473 0>, rot=<-0.0093 -0.707 0.0093 -0.707>}
 clavicle_l {parent=spine_05, loc=<4.3575 -0.1466 -2.3223>, rot=<0.0465 -0.7289 -0.009 -0.6829>}
 upperarm_l {parent=clavicle_l, loc=<15.5111 0 0>, rot=<0.0226 -0.3234 -0.0088 -0.9459>}
 lowerarm_l {parent=upperarm_l, loc=<26.6514 0 0>, rot=<0.0391 -0.0099 -0.2453 -0.9686>}
 hand_l {parent=lowerarm_l, loc=<22.3491 0 0>, rot=<-0.7081 -0.0507 0.0975 -0.6975>}
 thumb_01_l {parent=hand_l, loc=<1.3118 -0.3191 -2.459>, rot=<0.6546 -0.4042 0.0664 -0.6354>}
 thumb_02_l {parent=thumb_01_l, loc=<5.9336 -0.0448 0.0401>, rot=<-0.0037 0.0213 0.298 -0.9543>}
 thumb_03_l {parent=thumb_02_l, loc=<2.2223 0.3682 -0.0907>, rot=<0.0118 -0.013 0.0066 -0.9998>}
 index_metacarpal_l {parent=hand_l, loc=<1.6535 0.0183 -2.96>, rot=<0.0158 -0.0124 -0.0173 -0.9996>}
 index_01_l {parent=index_metacarpal_l, loc=<8.3008 0.0357 -0.0524>, rot=<-0.0044 0.0485 0.0924 -0.9945>}
 index_02_l {parent=index_01_l, loc=<3.3587 -0.3004 0.0452>, rot=<-0.002 0.0027 0.3225 -0.9466>}
 index_03_l {parent=index_02_l, loc=<2.5578 -0.123 -0.0252>, rot=<-0.002 0.0067 0.2558 -0.9667>}
 middle_metacarpal_l {parent=hand_l, loc=<2.1856 -0.1725 -0.5984>, rot=<0.006 0.0168 -0.0329 -0.9993>}
 middle_01_l {parent=middle_metacarpal_l, loc=<7.3275 -0.2447 0.0129>, rot=<-0.0208 0.0042 0.0858 -0.9961>}
 middle_02_l {parent=middle_01_l, loc=<4.0123 0.1401 0.1175>, rot=<-0.0218 0.0258 0.3512 -0.9357>}
 middle_03_l {parent=middle_02_l, loc=<2.6077 0.0534 0.1373>, rot=<0.0082 0.0188 0.1708 -0.9851>}
 ring_metacarpal_l {parent=hand_l, loc=<2.2074 -0.3625 1.3638>, rot=<-0.0284 0.0722 -0.0458 -0.9959>}
 ring_01_l {parent=ring_metacarpal_l, loc=<7.0295 -0.6197 -0.1446>, rot=<-0.0489 -0.0252 0.1892 -0.9804>}
 ring_02_l {parent=ring_01_l, loc=<3.6174 0.4647 -0.1027>, rot=<0 0 0.2806 -0.9598>}
 ring_03_l {parent=ring_02_l, loc=<2.4152 0.114 0.0199>, rot=<0 0 0.1189 -0.9929>}
 pinky_metacarpal_l {parent=hand_l, loc=<1.7616 -0.9753 3.3814>, rot=<-0.0397 0.0814 -0.069 -0.9935>}
 pinky_01_l {parent=pinky_metacarpal_l, loc=<7.1709 -0.4831 -0.1094>, rot=<-0.0654 -0.0035 0.2242 -0.9723>}
 pinky_02_l {parent=pinky_01_l, loc=<2.8274 0.0893 -0.1043>, rot=<0.0037 0.0011 0.2854 -0.9584>}
 pinky_03_l {parent=pinky_02_l, loc=<2.2046 0.0307 -0.0078>, rot=<-0.0002 0.0098 0.1445 -0.9895>}
 lowerarm_twist_02_l {parent=lowerarm_l, loc=<9.6199 0 0>, rot=<0 0 0 -1>}
 lowerarm_twist_01_l {parent=lowerarm_l, loc=<17.8549 0 0>, rot=<0 0 0 -1>}
 deform_elbow_in_base_l {parent=lowerarm_l, loc=<0.2112 5.259 0.8274>, rot=<0 0 0 -1>}
 deform_elbow_in_l {parent=deform_elbow_in_base_l, loc=<0 0 0>, rot=<0 0 0 1>}
 deform_elbow_out_base_l {parent=lowerarm_l, loc=<0 -4 0>, rot=<0 0 0 -1>}
 deform_elbow_out_l {parent=deform_elbow_out_base_l, loc=<0 0 0>, rot=<0 0 0 1>}
 upperarm_twist_01_l {parent=upperarm_l, loc=<0.0041 0 0>, rot=<0 0 0 1>}
 deform_upperarm_bk_base_l {parent=upperarm_twist_01_l, loc=<1.0813 -7.2193 -1.9453>, rot=<0 0 0 1>}
 deform_upperarm_bk_l {parent=deform_upperarm_bk_base_l, loc=<0 0 0>, rot=<0 0 0 1>}
 deform_upperarm_fr_base_l {parent=upperarm_twist_01_l, loc=<-0.3691 7.0506 -2.765>, rot=<0 0 0 1>}
 deform_upperarm_fr_l {parent=deform_upperarm_fr_base_l, loc=<0 0 0>, rot=<0 0 0 1>}
 deform_upperarm_up_base_l {parent=upperarm_twist_01_l, loc=<-3.2216 -1.1391 3.3311>, rot=<0 0 0 1>}
 deform_upperarm_up_l {parent=deform_upperarm_up_base_l, loc=<0 0 0>, rot=<0 0 0 1>}
 dyn_shoulderpad_l_MSTR {parent=upperarm_twist_01_l, loc=<-0.757 -1.6876 5.6896>, rot=<0.006 -0.0299 -0.0318 -0.999>}
 dyn_shoulderpad_l {parent=dyn_shoulderpad_l_MSTR, loc=<0.3 0 0>, rot=<0 0 0 1>}
 dyn_shoulderpad_dist_top_l_MSTR {parent=dyn_shoulderpad_l_MSTR, loc=<-8.2972 1.3556 -3.9817>, rot=<0 0 0 1>}
 dyn_shoulderpad_dist_top_l {parent=dyn_shoulderpad_dist_top_l_MSTR, loc=<0.3 0 0>, rot=<0 0 0 1>}
 upperarm_twist_02_l {parent=upperarm_l, loc=<13.7941 0 0>, rot=<0 0 0 1>}
 dyn_shoulderpad_up_target {parent=upperarm_twist_02_l, loc=<0 0 10.3552>, rot=<0 0 0 1>}
 dyn_shoulderpad_dist_bott_l_MSTR {parent=clavicle_l, loc=<10.2053 -0.7702 6.2989>, rot=<0.0321 -0.68 -0.0302 -0.7319>}
 dyn_shoulderpad_dist_bott_l {parent=dyn_shoulderpad_dist_bott_l_MSTR, loc=<0.3 0 0>, rot=<0 0 0 1>}
 clavicle_r {parent=spine_05, loc=<4.3575 -0.1466 2.3223>, rot=<0.7289 0.0465 -0.6829 0.009>}
 upperarm_r {parent=clavicle_r, loc=<-15.5111 0 0>, rot=<0.0226 -0.3234 -0.0088 -0.9459>}
 lowerarm_r {parent=upperarm_r, loc=<-26.6514 0 0>, rot=<0.0391 -0.0099 -0.2453 -0.9686>}
 hand_r {parent=lowerarm_r, loc=<-22.3491 0 0>, rot=<-0.7081 -0.0507 0.0975 -0.6975>}
 thumb_01_r {parent=hand_r, loc=<-1.3118 0.3191 2.459>, rot=<0.6546 -0.4042 0.0664 -0.6354>}
 thumb_02_r {parent=thumb_01_r, loc=<-5.9336 0.0448 -0.0401>, rot=<-0.0037 0.0213 0.298 -0.9543>}
 thumb_03_r {parent=thumb_02_r, loc=<-2.2223 -0.3682 0.0907>, rot=<0.0118 -0.013 0.0066 -0.9998>}
 index_metacarpal_r {parent=hand_r, loc=<-1.6535 -0.0183 2.96>, rot=<0.0158 -0.0124 -0.0173 -0.9996>}
 index_01_r {parent=index_metacarpal_r, loc=<-8.3008 -0.0357 0.0524>, rot=<-0.0044 0.0485 0.0924 -0.9945>}
 index_02_r {parent=index_01_r, loc=<-3.3587 0.3004 -0.0452>, rot=<-0.002 0.0027 0.3225 -0.9466>}
 index_03_r {parent=index_02_r, loc=<-2.5578 0.123 0.0252>, rot=<-0.002 0.0067 0.2558 -0.9667>}
 middle_metacarpal_r {parent=hand_r, loc=<-2.1856 0.1725 0.5984>, rot=<0.006 0.0168 -0.0329 -0.9993>}
 middle_01_r {parent=middle_metacarpal_r, loc=<-7.3275 0.2447 -0.0129>, rot=<-0.0208 0.0042 0.0858 -0.9961>}
 middle_02_r {parent=middle_01_r, loc=<-4.0123 -0.1401 -0.1175>, rot=<-0.0218 0.0258 0.3512 -0.9357>}
 middle_03_r {parent=middle_02_r, loc=<-2.6077 -0.0534 -0.1373>, rot=<0.0082 0.0188 0.1708 -0.9851>}
 ring_metacarpal_r {parent=hand_r, loc=<-2.2074 0.3625 -1.3638>, rot=<-0.0284 0.0722 -0.0458 -0.9959>}
 ring_01_r {parent=ring_metacarpal_r, loc=<-7.0295 0.6197 0.1446>, rot=<-0.0489 -0.0252 0.1892 -0.9804>}
 ring_02_r {parent=ring_01_r, loc=<-3.6174 -0.4647 0.1027>, rot=<0 0 0.2806 -0.9598>}
 ring_03_r {parent=ring_02_r, loc=<-2.4152 -0.114 -0.0199>, rot=<0 0 0.1189 -0.9929>}
 pinky_metacarpal_r {parent=hand_r, loc=<-1.7616 0.9753 -3.3814>, rot=<-0.0397 0.0814 -0.069 -0.9935>}
 pinky_01_r {parent=pinky_metacarpal_r, loc=<-7.1709 0.4831 0.1094>, rot=<-0.0654 -0.0035 0.2242 -0.9723>}
 pinky_02_r {parent=pinky_01_r, loc=<-2.8274 -0.0893 0.1043>, rot=<0.0037 0.0011 0.2854 -0.9584>}
 pinky_03_r {parent=pinky_02_r, loc=<-2.2046 -0.0307 0.0078>, rot=<-0.0002 0.0098 0.1445 -0.9895>}
 lowerarm_twist_02_r {parent=lowerarm_r, loc=<-9.6199 0 0>, rot=<0 0 0 -1>}
 lowerarm_twist_01_r {parent=lowerarm_r, loc=<-17.8549 0 0>, rot=<0 0 0 -1>}
 deform_elbow_in_base_r {parent=lowerarm_r, loc=<-0.2111 -5.2591 -0.8275>, rot=<0 0 0 -1>}
 deform_elbow_in_r {parent=deform_elbow_in_base_r, loc=<-0.0001 0 0.0002>, rot=<0 0 0 1>}
 deform_elbow_out_base_r {parent=lowerarm_r, loc=<0 4 -0.0001>, rot=<0 0 0 -1>}
 deform_elbow_out_r {parent=deform_elbow_out_base_r, loc=<0 0 0>, rot=<0 0 0 1>}
 upperarm_twist_01_r {parent=upperarm_r, loc=<-0.0041 0 0>, rot=<0 0 0 -1>}
 dyn_pauldron_strap {parent=upperarm_twist_01_r, loc=<-1.8857 0.73 0.1851>, rot=<0.0217 -0.9151 -0.0305 0.4014>}
 deform_upperarm_bk_base_r {parent=upperarm_twist_01_r, loc=<-1.0814 7.2192 1.9455>, rot=<0 0 0 -1>}
 deform_upperarm_bk_r {parent=deform_upperarm_bk_base_r, loc=<0 0 0>, rot=<0 0 0 1>}
 deform_upperarm_fr_base_r {parent=upperarm_twist_01_r, loc=<0.3689 -7.0506 2.7652>, rot=<0 0 0 -1>}
 deform_upperarm_fr_r {parent=deform_upperarm_fr_base_r, loc=<0 0 0>, rot=<0 0 0 1>}
 deform_upperarm_up_base_r {parent=upperarm_twist_01_r, loc=<3.2214 1.1391 -3.3308>, rot=<0 0 0 -1>}
 deform_upperarm_up_r {parent=deform_upperarm_up_base_r, loc=<0 0 0>, rot=<0 0 0 -1>}
 dyn_pauldron_bk_bott_strap_pin_r_MSTR {parent=upperarm_twist_01_r, loc=<-6.2871 4.7653 5.129>, rot=<-0.4849 -0.7079 -0.5028 0.1044>}
 dyn_pauldron_bk_bott_strap_pin_r {parent=dyn_pauldron_bk_bott_strap_pin_r_MSTR, loc=<2.359 0 0>, rot=<0 0 0 1>}
 dyn_pauldron_fr_bott_strap_pin_r_MSTR {parent=upperarm_twist_01_r, loc=<-5.6369 -5.0244 4.8759>, rot=<0.4655 -0.7713 0.4328 0.0333>}
 dyn_pauldron_fr_bott_strap_pin_r {parent=dyn_pauldron_fr_bott_strap_pin_r_MSTR, loc=<1.0189 0 0>, rot=<0 0 0 1>}
 upperarm_twist_02_r {parent=upperarm_r, loc=<-13.7941 0 0>, rot=<0 0 0 -1>}
 upperarm_r_ctrl {controlType=EULER_TRANSFORM, parent=upperarm_r, type=Control, loc=<0 0 0>, rot=<0 0 0 1>}
 dyn_pauldron_fr_01_r_MSTR {parent=clavicle_r, loc=<-2.8477 -5.3717 -6.5957>, rot=<0.0576 -0.9503 0.2862 0.1085>}
 dyn_pauldron_fr_01_r {parent=dyn_pauldron_fr_01_r_MSTR, loc=<0.1649 -0.2505 -0.0053>, rot=<0.0448 -0.0342 0.4732 -0.8791>}
 dyn_pauldron_fr_02_r_MSTR {parent=dyn_pauldron_fr_01_r_MSTR, loc=<7.3456 0.4659 0.3987>, rot=<0.0204 -0.0526 -0.0544 -0.9969>}
 dyn_pauldron_fr_02_r {parent=dyn_pauldron_fr_02_r_MSTR, loc=<0.3 0 0>, rot=<0 0 0 1>}
 dyn_pauldron_fr_03_r_MSTR {parent=dyn_pauldron_fr_02_r_MSTR, loc=<7.2894 0.2439 1.0121>, rot=<0 0 0 1>}
 dyn_pauldron_fr_03_r {parent=dyn_pauldron_fr_03_r_MSTR, loc=<0.3 0 0>, rot=<0 0 0 1>}
 dyn_pauldron_fr_04_r_MSTR {parent=dyn_pauldron_fr_03_r_MSTR, loc=<11.1789 -0.892 1.8488>, rot=<-0.0848 0.0865 0.0074 -0.9926>}
 dyn_pauldron_fr_04_r {parent=dyn_pauldron_fr_04_r_MSTR, loc=<0.3 0 0>, rot=<0 0 0 1>}
 dyn_pauldron_fr_08_r_MSTR {parent=dyn_pauldron_fr_03_r_MSTR, loc=<4.0009 -5.5056 -2.0211>, rot=<-0.2453 -0.0684 0.4985 -0.8287>}
 dyn_pauldron_fr_08_r {parent=dyn_pauldron_fr_08_r_MSTR, loc=<0.3 0 0>, rot=<0 0 0 1>}
 dyn_pauldron_fr_09_r_MSTR {parent=dyn_pauldron_fr_03_r_MSTR, loc=<8.5408 -4.3093 -0.2962>, rot=<-0.3006 -0.0198 0.346 -0.8886>}
 dyn_pauldron_fr_09_r {parent=dyn_pauldron_fr_09_r_MSTR, loc=<0.3 0 0>, rot=<0 0 0 1>}
 dyn_pauldron_fr_06_r_MSTR {parent=dyn_pauldron_fr_02_r_MSTR, loc=<6.1105 -4.8762 -0.7585>, rot=<-0.1569 -0.0019 0.5749 -0.803>}
 dyn_pauldron_fr_06_r {parent=dyn_pauldron_fr_06_r_MSTR, loc=<0.3 0 0>, rot=<0 0 0 1>}
 dyn_pauldron_fr_strap_pin_r_MSTR {parent=dyn_pauldron_fr_06_r, loc=<0.4121 0 0>, rot=<0 0 0 1>}
 dyn_pauldron_fr_strap_pin_r {parent=dyn_pauldron_fr_strap_pin_r_MSTR, loc=<0.3 0 0>, rot=<0 0 0 1>}
 dyn_pauldron_fr_07_r_MSTR {parent=dyn_pauldron_fr_02_r_MSTR, loc=<1.8389 -5.6066 -0.7594>, rot=<0.02 -0.0157 0.6473 -0.7618>}
 dyn_pauldron_fr_07_r {parent=dyn_pauldron_fr_07_r_MSTR, loc=<0.3 0 0>, rot=<0 0 0 1>}
 dyn_pauldron_fr_05_r_MSTR {parent=dyn_pauldron_fr_01_r_MSTR, loc=<5.851 -6.2079 0.2444>, rot=<0.0463 -0.0321 0.5131 -0.8565>}
 dyn_pauldron_fr_05_r {parent=dyn_pauldron_fr_05_r_MSTR, loc=<0.3 0 0>, rot=<0 0 0 1>}
 dyn_pauldron_bk_01_r_MSTR {parent=clavicle_r, loc=<-2.7203 4.4136 -9.9046>, rot=<-0.3175 -0.9224 -0.1397 0.1698>}
 dyn_pauldron_bk_01_r {parent=dyn_pauldron_bk_01_r_MSTR, loc=<0.2427 0.1742 -0.0274>, rot=<0.1143 -0.0109 -0.3088 -0.9442>}
 dyn_pauldron_bk_02_r_MSTR {parent=dyn_pauldron_bk_01_r_MSTR, loc=<9.5018 -1.9316 -0.4001>, rot=<0.2171 -0.1887 -0.0393 -0.9569>}
 dyn_pauldron_bk_02_r {parent=dyn_pauldron_bk_02_r_MSTR, loc=<0.3 0 0>, rot=<0 0 0 1>}
 dyn_pauldron_bk_03_r_MSTR {parent=dyn_pauldron_bk_02_r_MSTR, loc=<9.2575 -2.3222 -0.4312>, rot=<0 0 -0.3646 -0.9312>}
 dyn_pauldron_bk_03_r {parent=dyn_pauldron_bk_03_r_MSTR, loc=<0.3 0 0>, rot=<0 0 0 1>}
 dyn_pauldron_bk_strap_pin_r_MSTR {parent=dyn_pauldron_bk_03_r, loc=<-0.3 0 0.3776>, rot=<0 0 0 1>}
 dyn_pauldron_bk_strap_pin_r {parent=dyn_pauldron_bk_strap_pin_r_MSTR, loc=<0.3 0 0>, rot=<0 0 0 1>}
 dyn_pauldron_top_01_r_MSTR {parent=clavicle_r, loc=<-4.2569 -0.2593 -9.3025>, rot=<-0.0063 -0.994 0.0779 0.0765>}
 dyn_pauldron_top_01_r {parent=dyn_pauldron_top_01_r_MSTR, loc=<0.2996 -0.0155 -0.0014>, rot=<-0.027 -0.0017 0.0258 -0.9993>}
 dyn_pauldron_top_02_r_MSTR {parent=dyn_pauldron_top_01_r_MSTR, loc=<8.0702 0.1259 0.236>, rot=<0.0685 -0.0053 0.0088 -0.9976>}
 dyn_pauldron_top_02_r {parent=dyn_pauldron_top_02_r_MSTR, loc=<0.3 0 0>, rot=<0 0 0 1>}
 dyn_pauldron_top_03_r_MSTR {parent=dyn_pauldron_top_02_r_MSTR, loc=<6.867 0.1611 -0.0634>, rot=<-0.0956 0 0 -0.9954>}
 dyn_pauldron_top_03_r {parent=dyn_pauldron_top_03_r_MSTR, loc=<0.3 0 0>, rot=<0 0 0 1>}
 dyn_pauldron_top_04_r_MSTR {parent=dyn_pauldron_top_03_r_MSTR, loc=<11.0539 0 -0.3408>, rot=<0 0 0 1>}
 dyn_pauldron_top_04_r {parent=dyn_pauldron_top_04_r_MSTR, loc=<0.3 0 0>, rot=<0 0 0 1>}
 dyn_pauldron_top_05_r_MSTR {parent=dyn_pauldron_top_03_r_MSTR, loc=<11.2149 -3.5644 -0.302>, rot=<-0.0681 0.015 0.0377 -0.9969>}
 dyn_pauldron_top_05_r {parent=dyn_pauldron_top_05_r_MSTR, loc=<0.3 0 0>, rot=<0 0 0 1>}
 dyn_pauldron_topbk_01_r_MSTR {parent=dyn_pauldron_top_01_r_MSTR, loc=<7.7877 4.3506 -0.7758>, rot=<0.3262 -0.1299 -0.055 -0.9347>}
 dyn_pauldron_topbk_01_r {parent=dyn_pauldron_topbk_01_r_MSTR, loc=<0.3 0 0>, rot=<0 0 0 1>}
 dyn_pauldron_topbk_02_r_MSTR {parent=dyn_pauldron_topbk_01_r_MSTR, loc=<8.7578 -0.1986 1.2607>, rot=<-0.0098 0.0444 -0.0241 -0.9987>}
 dyn_pauldron_topbk_02_r {parent=dyn_pauldron_topbk_02_r_MSTR, loc=<0.3 0 0>, rot=<0 0 0 1>}
 dyn_pauldron_topbk_03_r_MSTR {parent=dyn_pauldron_topbk_02_r_MSTR, loc=<8.086 0 0>, rot=<0 0 0 1>}
 dyn_pauldron_topbk_03_r {parent=dyn_pauldron_topbk_03_r_MSTR, loc=<0.2994 0.0193 -0.002>, rot=<0 0 0 1>}
 spline_01_ctrl {controlType=EULER_TRANSFORM, parent=clavicle_r, type=Control, loc=<0 0 0>, rot=<0 0 0 1>}
 spline_02_ctrl {controlType=EULER_TRANSFORM, parent=clavicle_r, type=Control, loc=<0 0 0>, rot=<0 0 0 1>}
 spline_03_ctrl {controlType=EULER_TRANSFORM, parent=clavicle_r, type=Control, loc=<0 0 0>, rot=<0 0 0 1>}
 spline_04_ctrl {controlType=EULER_TRANSFORM, parent=clavicle_r, type=Control, loc=<0 0 0>, rot=<0 0 0 1>}
 dyn_chest_rig_l_MSTR {parent=spine_05, loc=<3.2496 9.812 -11.9838>, rot=<-0.1692 -0.7388 -0.6519 0.0234>}
 dyn_chest_rig_l {parent=dyn_chest_rig_l_MSTR, loc=<0.3 0 0>, rot=<0 0 0 1>}
 dyn_belly_rig_pouch_r_03 {parent=spine_01, loc=<5.0739 13.3622 2.6253>, rot=<0 0.6261 0.7797 0>}
 dyn_grenade_r_MSTR {parent=dyn_belly_rig_pouch_r_03, loc=<0.7999 1.7137 4.6492>, rot=<-0.0091 -0.1757 0.0071 -0.9844>}
 dyn_grenade_r {parent=dyn_grenade_r_MSTR, loc=<0.3 0 0>, rot=<0 0 0 1>}
 dyn_belly_rig_pouch_l_03 {parent=spine_01, loc=<5.0739 13.3622 -2.6253>, rot=<-0.0331 0.7786 0.6252 -0.0412>}
 dyn_grenade_m_MSTR {parent=dyn_belly_rig_pouch_l_03, loc=<1.9085 3.8079 4.0205>, rot=<0.1583 -0.0669 0.0498 -0.9839>}
 dyn_grenade_m {parent=dyn_grenade_m_MSTR, loc=<0.3 0 0>, rot=<0 0 0 1>}
 dyn_grenade_l_MSTR {parent=dyn_belly_rig_pouch_l_03, loc=<1.0136 -0.5631 4.6728>, rot=<0.0634 -0.0828 0.0357 -0.9939>}
 dyn_grenade_l {parent=dyn_grenade_l_MSTR, loc=<0.3 0 0>, rot=<0 0 0 1>}
 dyn_belly_rig_pouch_r_01_MSTR {parent=spine_01, loc=<5.8989 6.1041 11.7236>, rot=<0.0235 0.2296 0.973 -0.0047>}
 dyn_belly_rig_pouch_r_01 {parent=dyn_belly_rig_pouch_r_01_MSTR, loc=<0.3 0 0>, rot=<0 0 0 1>}
 dyn_belly_rig_pouch_r_02_MSTR {parent=spine_01, loc=<5.7358 10.467 8.1671>, rot=<0.0216 0.4652 0.8849 -0.0104>}
 dyn_belly_rig_pouch_r_02 {parent=dyn_belly_rig_pouch_r_02_MSTR, loc=<0.3 0 0>, rot=<0 0 0 1>}
 dyn_back_canister_MSTR {parent=spine_01, loc=<3.4916 -10.3443 -1.8543>, rot=<-0.0349 -0.7022 0.7111 -0.0056>}
 dyn_back_canister {parent=dyn_back_canister_MSTR, loc=<0.3 0 0>, rot=<0 0 0 1>}
 dyn_belly_rig_pouch_l_02_MSTR {parent=spine_01, loc=<5.7357 10.4645 -8.1654>, rot=<0.0081 -0.8848 -0.4652 -0.0261>}
 dyn_belly_rig_pouch_l_02 {parent=dyn_belly_rig_pouch_l_02_MSTR, loc=<0.3 0 0>, rot=<0 0 0 1>}
 dyn_belly_rig_pouch_l_01_MSTR {parent=spine_01, loc=<5.8988 6.1027 -11.7209>, rot=<0.0035 -0.9728 -0.2297 -0.0284>}
 dyn_belly_rig_pouch_l_01 {parent=dyn_belly_rig_pouch_l_01_MSTR, loc=<0.3 0 0>, rot=<0 0 0 1>}
 thigh_l {parent=pelvis, loc=<-2.5693 -0.0109 -7.7501>, rot=<0.0012 0.0572 0.0155 -0.9982>}
 calf_l {parent=thigh_l, loc=<-39.0484 0 0>, rot=<0.0006 -0.0161 -0.0391 -0.9991>}
 foot_l {parent=calf_l, loc=<-42.561 0.1077 0.4292>, rot=<0.0418 -0.0455 0.0343 -0.9975>}
 ball_l {parent=foot_l, loc=<-6.1599 13.8264 0.6806>, rot=<0.0001 -0.0001 -0.7063 -0.7079>}
 calf_twist_02_l {parent=calf_l, loc=<-18.8064 -0.0967 0.0397>, rot=<0.0237 0.0095 -0.0035 -0.9997>}
 calf_twist_01_l {parent=calf_l, loc=<-35.813 -0.0967 0.0397>, rot=<0.0237 0.0095 -0.0035 -0.9997>}
 deform_knee_in_base_l {parent=calf_l, loc=<1.298 -5.2957 -0.0496>, rot=<0 0 0 -1>}
 deform_knee_in_l {parent=deform_knee_in_base_l, loc=<0 0 0>, rot=<0 0 0 1>}
 deform_knee_out_base_l {parent=calf_l, loc=<0 4 0>, rot=<0 0 0 -1>}
 deform_knee_out_l {parent=deform_knee_out_base_l, loc=<0 0 0>, rot=<0 0 0 1>}
 thigh_twist_01_l {parent=thigh_l, loc=<-21.6303 0 0>, rot=<0.0013 0 0 -1>}
 deform_groin_base_l {parent=thigh_l, loc=<-2.1908 8.1142 -1.7442>, rot=<0 0 0 -1>}
 deform_groin_l {parent=deform_groin_base_l, loc=<0 0 0>, rot=<0 0 0 1>}
 thigh_r {parent=pelvis, loc=<-2.5693 -0.0109 7.7501>, rot=<0.0572 -0.0012 0.9982 0.0155>}
 calf_r {parent=thigh_r, loc=<39.0484 0 0>, rot=<0.0006 -0.0161 -0.0391 -0.9991>}
 foot_r {parent=calf_r, loc=<42.561 -0.1077 -0.4292>, rot=<0.0418 -0.0455 0.0343 -0.9975>}
 ball_r {parent=foot_r, loc=<6.1599 -13.8264 -0.6806>, rot=<0.0001 -0.0001 -0.7063 -0.7079>}
 calf_twist_02_r {parent=calf_r, loc=<18.8064 0.0967 -0.0397>, rot=<0.0237 0.0095 -0.0035 -0.9997>}
 calf_twist_01_r {parent=calf_r, loc=<35.813 0.0967 -0.0397>, rot=<0.0237 0.0095 -0.0035 -0.9997>}
 deform_knee_in_base_r {parent=calf_r, loc=<-1.298 5.2957 0.0496>, rot=<0 0 0 -1>}
 deform_knee_in_r {parent=deform_knee_in_base_r, loc=<0 0 0>, rot=<0 0 0 1>}
 deform_knee_out_base_r {parent=calf_r, loc=<0 -4 0>, rot=<0 0 0 -1>}
 deform_knee_out_r {parent=deform_knee_out_base_r, loc=<0 0 0>, rot=<0 0 0 1>}
 thigh_twist_01_r {parent=thigh_r, loc=<21.6303 0 0>, rot=<0.0013 0 0 -1>}
 deform_groin_base_r {parent=thigh_r, loc=<2.1908 -8.1142 1.7442>, rot=<0 0 0 -1>}
 deform_groin_r {parent=deform_groin_base_r, loc=<0 0 0>, rot=<0 0 0 1>}
 deform_glute_base_l {parent=pelvis, loc=<-8.8064 -8.8222 -8.4287>, rot=<0.0275 0.0501 -0.4472 -0.8926>}
 deform_glute_l {parent=deform_glute_base_l, loc=<0 0 0>, rot=<0 0 0 1>}
 deform_glute_base_r {parent=pelvis, loc=<-8.8062 -8.8222 8.4287>, rot=<0.0501 -0.0275 0.8926 -0.4472>}
 deform_glute_r {parent=deform_glute_base_r, loc=<0 0 0>, rot=<0 0 0 1>}
 dyn_se14_holster_base_MSTR {parent=pelvis, loc=<-0.0165 3.6991 16.9287>, rot=<0.1181 0.1654 0.979 0.0141>}
 dyn_se14_holster_base {parent=dyn_se14_holster_base_MSTR, loc=<0.3 0 0>, rot=<0 0 0 1>}
 dyn_se14_holster_loop {parent=dyn_se14_holster_base, loc=<3.8979 0 0.7542>, rot=<0 -0.0302 0 0.9995>}
 dyn_se14_holster_strap {parent=dyn_se14_holster_base, loc=<6.942 -0.3104 -0.4526>, rot=<0 0.1043 0 0.9945>}
 dyn_belt_pouch_06_MSTR {parent=pelvis, loc=<2.1308 -3.7715 16.3615>, rot=<0.1288 -0.1436 0.9811 -0.0173>}
 dyn_belt_pouch_06 {parent=dyn_belt_pouch_06_MSTR, loc=<0.3 0 0>, rot=<0 0 0 1>}
 dyn_belt_pouch_04_MSTR {parent=pelvis, loc=<4.4511 -9.4458 11.9867>, rot=<0.0034 -0.3843 0.9232 -0.0045>}
 dyn_belt_pouch_04 {parent=dyn_belt_pouch_04_MSTR, loc=<0.3 0 0>, rot=<0 0 0 1>}
 dyn_belt_pouch_01_MSTR {parent=pelvis, loc=<4.363 10.6318 -7.8025>, rot=<0.0193 -0.7905 -0.6119 -0.0148>}
 dyn_belt_pouch_01 {parent=dyn_belt_pouch_01_MSTR, loc=<0.3 0 0>, rot=<0 0 0 1>}
 dyn_belt_pouch_02_MSTR {parent=pelvis, loc=<2.4114 7.7678 -14.0205>, rot=<0.0074 -0.94 -0.3408 -0.0116>}
 dyn_belt_pouch_02 {parent=dyn_belt_pouch_02_MSTR, loc=<0.3 0 0>, rot=<0 0 0 1>}
 dyn_belt_pouch_03_MSTR {parent=pelvis, loc=<2.7689 -8.087 -13.4638>, rot=<0.0002 -0.9304 0.3567 -0.0844>}
 dyn_belt_pouch_03 {parent=dyn_belt_pouch_03_MSTR, loc=<0.3 0 0>, rot=<0 0 0 1>}
 dyn_belt_pouch_05_MSTR {parent=pelvis, loc=<4.3958 10.2589 8.563>, rot=<0.0043 0.572 0.8202 -0.0029>}
 dyn_belt_pouch_05 {parent=dyn_belt_pouch_05_MSTR, loc=<0.3 0 0>, rot=<0 0 0 1>}
]
variables [
 ParentConstraint_Weight {type=float, default=1}
 SplineUValue {type=float, default=0}
]
