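# Physics asset for the SageTwig medium male player body, covering the SE14 holster's dynamic bones and the right thigh.
# Capsules shape the holster base, loop and strap, with a tiny capsule on the master bone.
# Constraints link the base to the master, and the loop and strap to the base, sharing narrow swing and twist limits.
# The unconstrained thigh carries two nearly coincident capsules, apparently as a collider for the holster.
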PhysicsAsset /BRCosmetics/Characters/Player/Male/Medium/Bodies/M_MED_SageTwig/Meshes/M_MED_SageTwig_Physics
bodies {
 dyn_se14_holster_base [{
   shapeType=Capsule
   length=4.6432
   radius=5.8887
   shapeName=dyn_se14_holster_base_capsule
   center=<1.7673 0 4.9212>
   rotation=<rot 67.5525 179.9999 179.9999>
  }]
 dyn_se14_holster_base_MSTR [{shapeType=Capsule, length=0.1, radius=0.1, shapeName=dyn_se14_holster_base_MSTR_capsule, rotation=<rot 90 0 0>}]
 dyn_se14_holster_loop [{shapeType=Capsule, length=5, radius=2.8973, shapeName=dyn_se14_holster_base_capsule, center=<3 0 2.1678>, rotation=<rot 90 0 0>}]
 dyn_se14_holster_strap [{shapeType=Capsule, length=12.1993, radius=5.8887, shapeName=dyn_se14_holster_base_capsule, center=<3 0 5.5383>, rotation=<rot 90 0 0>}]
 thigh_r [
  {
   shapeType=Capsule
   length=27.2223
   radius=8.77
   shapeName=thigh_r_capsule
   center=<18.6121 0.3776 -0.874>
   rotation=<rot 63.2968 -72.5742 -80.1541>
  }
  {
   shapeType=Capsule
   length=27.2223
   radius=8.77
   shapeName=thigh_r_capsule
   center=<18.8432 1.0357 -0.8947>
   rotation=<rot 65.2389 -70.7603 -72.7091>
  }
 ]
}
constraints [
 {bone1Name=dyn_se14_holster_base, bone2Name=dyn_se14_holster_base_MSTR, swing1LimitDegrees=10, swing2LimitDegrees=30, twistLimitDegrees=2}
 {bone1Name=dyn_se14_holster_loop, bone2Name=dyn_se14_holster_base, swing1LimitDegrees=10, swing2LimitDegrees=30, twistLimitDegrees=2}
 {bone1Name=dyn_se14_holster_strap, bone2Name=dyn_se14_holster_base, swing1LimitDegrees=10, swing2LimitDegrees=30, twistLimitDegrees=2}
]
