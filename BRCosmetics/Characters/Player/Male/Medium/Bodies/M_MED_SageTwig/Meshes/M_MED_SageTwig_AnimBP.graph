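# Post-process style AnimBlueprint for the M_MED_SageTwig player body cosmetic, adding procedural and secondary motion to a pose copied from the attached parent mesh.
# The copied pose runs through the body's Control Rig, then chained Anim Dynamics nodes spring grenade, belt pouch and left chest rig bones.
# A final RigidBody node simulates the override physics asset in pelvis space.
# The EventGraph holds upper-arm joint debug draws, unconnected to the update event.
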AnimBlueprint /BRCosmetics/Characters/Player/Male/Medium/Bodies/M_MED_SageTwig/Meshes/M_MED_SageTwig_AnimBP
target_skeleton=/Game/Characters/Player/Male/Medium/Bodies/M_MED_SageTwig/Meshes/M_MED_SageTwig_Skeleton
graphs [
 AnimGraph {nodes [
   n1 = Control Rig(Source=Copy Pose From Mesh() {bUseAttachedParent=true}) {
    alphaBoolBlend {blendOption=Linear}
    controlRigAssetReference {blueprintRigClass=&/M_MED_SageTwig_ControlRig.M_MED_SageTwig_ControlRig_C}
   }
   n2 = Anim Dynamics - Bone: dyn_grenade_l(ComponentPose=Local To Component(LocalPose=@n1), Alpha=1, ExternalForce="0,0,0") {
    angularDampingOverride=1
    angularSpringConstant=20000
    bAngularSpring=true
    bOverrideAngularDamping=true
    bUseGravityOverride=true
    gravityScale=0.35
    simulationSpace=RootRelative
    alphaBoolBlend {blendOption=Linear}
    boundBone {boneName=dyn_grenade_l}
    componentAppliedLinearAccClamp=<1000 1000 1000>
    componentLinearAccScale=<1 1 1>
    componentLinearVelScale=<1 1 1>
    gravityOverride {z=-980}
    physicsBodyDefinitions [{
      boundBone {boneName=dyn_grenade_l}
      boxExtents=<30 22 31>
      constraintSetup {angularLimitsMax=<15 5 15>, angularLimitsMin=<-15 -5 -15>, angularTarget=<1 0 0>}
      localJointOffset=<3.6 0 0>
     }]
   }
   n3 = Anim Dynamics - Bone: dyn_grenade_r(ComponentPose=@n2, Alpha=1, ExternalForce="0,0,0") {
    angularDampingOverride=1
    angularSpringConstant=20000
    bAngularSpring=true
    bOverrideAngularDamping=true
    bUseGravityOverride=true
    gravityScale=0.35
    simulationSpace=RootRelative
    alphaBoolBlend {blendOption=Linear}
    boundBone {boneName=dyn_grenade_r}
    componentAppliedLinearAccClamp=<1000 1000 1000>
    componentLinearAccScale=<1 1 1>
    componentLinearVelScale=<1 1 1>
    gravityOverride {z=-980}
    physicsBodyDefinitions [{
      boundBone {boneName=dyn_grenade_r}
      boxExtents=<40 30 30>
      constraintSetup {angularLimitsMax=<15 5 15>, angularLimitsMin=<-15 -5 -15>, angularTarget=<1 0 0>}
      localJointOffset=<4.5 0 0>
     }]
   }
   n4 = Anim Dynamics - Bone: dyn_grenade_m(ComponentPose=@n3, Alpha=1, ExternalForce="0,0,0") {
    angularDampingOverride=1
    angularSpringConstant=20000
    bAngularSpring=true
    bOverrideAngularDamping=true
    bUseGravityOverride=true
    gravityScale=0.35
    simulationSpace=RootRelative
    alphaBoolBlend {blendOption=Linear}
    boundBone {boneName=dyn_grenade_m}
    componentAppliedLinearAccClamp=<1000 1000 1000>
    componentLinearAccScale=<1 1 1>
    componentLinearVelScale=<1 1 1>
    gravityOverride {z=-980}
    physicsBodyDefinitions [{
      boundBone {boneName=dyn_grenade_m}
      boxExtents=<38 33 23>
      constraintSetup {angularLimitsMax=<15 5 15>, angularLimitsMin=<-15 -5 -15>, angularTarget=<1 0 0>}
      localJointOffset=<3.24 0 0>
     }]
   }
   n5 = Anim Dynamics - Bone: dyn_belt_pouch_01(ComponentPose=@n4, Alpha=1, ExternalForce="0,0,0") {
    angularDampingOverride=1
    angularSpringConstant=20000
    bAngularSpring=true
    bOverrideAngularDamping=true
    bUseGravityOverride=true
    gravityScale=0.35
    simulationSpace=RootRelative
    alphaBoolBlend {blendOption=Linear}
    boundBone {boneName=dyn_belt_pouch_01}
    componentAppliedLinearAccClamp=<1000 1000 1000>
    componentLinearAccScale=<1 1 1>
    componentLinearVelScale=<1 1 1>
    gravityOverride {z=-980}
    physicsBodyDefinitions [{
      boundBone {boneName=dyn_belt_pouch_01}
      boxExtents=<20 30 30>
      constraintSetup {angularLimitsMax=<3 5 5>, angularLimitsMin=<-3 -5 -5>, angularTarget=<1 0 0>}
      localJointOffset=<4 0 0>
     }]
   }
   n6 = Anim Dynamics - Bone: dyn_belt_pouch_02(ComponentPose=@n5, Alpha=1, ExternalForce="0,0,0") {
    angularDampingOverride=1
    angularSpringConstant=20000
    bAngularSpring=true
    bOverrideAngularDamping=true
    bUseGravityOverride=true
    gravityScale=0.35
    simulationSpace=RootRelative
    alphaBoolBlend {blendOption=Linear}
    boundBone {boneName=dyn_belt_pouch_02}
    componentAppliedLinearAccClamp=<1000 1000 1000>
    componentLinearAccScale=<1 1 1>
    componentLinearVelScale=<1 1 1>
    gravityOverride {z=-980}
    physicsBodyDefinitions [{
      boundBone {boneName=dyn_belt_pouch_02}
      boxExtents=<26 28 33>
      constraintSetup {angularLimitsMax=<3 5 5>, angularLimitsMin=<-3 -5 -5>, angularTarget=<1 0 0>}
      localJointOffset=<3.7 0 0>
     }]
   }
   n7 = Anim Dynamics - Bone: dyn_belt_pouch_04(ComponentPose=@n6, Alpha=1, ExternalForce="0,0,0") {
    angularDampingOverride=1
    angularSpringConstant=20000
    bAngularSpring=true
    bOverrideAngularDamping=true
    bUseGravityOverride=true
    gravityScale=0.35
    simulationSpace=RootRelative
    alphaBoolBlend {blendOption=Linear}
    boundBone {boneName=dyn_belt_pouch_04}
    componentAppliedLinearAccClamp=<1000 1000 1000>
    componentLinearAccScale=<1 1 1>
    componentLinearVelScale=<1 1 1>
    gravityOverride {z=-980}
    physicsBodyDefinitions [{
      boundBone {boneName=dyn_belt_pouch_04}
      boxExtents=<18 35 24>
      constraintSetup {angularLimitsMax=<3 5 5>, angularLimitsMin=<-3 -5 -5>, angularTarget=<1 0 0>}
      localJointOffset=<4 0 0>
     }]
   }
   n8 = Anim Dynamics - Bone: dyn_belt_pouch_05(ComponentPose=@n7, Alpha=1, ExternalForce="0,0,0") {
    angularDampingOverride=1
    angularSpringConstant=20000
    bAngularSpring=true
    bOverrideAngularDamping=true
    bUseGravityOverride=true
    gravityScale=0.35
    simulationSpace=RootRelative
    alphaBoolBlend {blendOption=Linear}
    boundBone {boneName=dyn_belt_pouch_05}
    componentAppliedLinearAccClamp=<1000 1000 1000>
    componentLinearAccScale=<1 1 1>
    componentLinearVelScale=<1 1 1>
    gravityOverride {z=-980}
    physicsBodyDefinitions [{
      boundBone {boneName=dyn_belt_pouch_05}
      boxExtents=<20 30 30>
      constraintSetup {angularLimitsMax=<3 5 5>, angularLimitsMin=<-3 -5 -5>, angularTarget=<1 0 0>}
      localJointOffset=<4 0 0>
     }]
   }
   n9 = Anim Dynamics - Bone: dyn_chest_rig_l(ComponentPose=@n8, Alpha=1, ExternalForce="0,0,0") {
    angularDampingOverride=1
    angularSpringConstant=20000
    bAngularSpring=true
    bOverrideAngularDamping=true
    bUseGravityOverride=true
    gravityScale=0.35
    simulationSpace=RootRelative
    alphaBoolBlend {blendOption=Linear}
    boundBone {boneName=dyn_chest_rig_l}
    componentAppliedLinearAccClamp=<1000 1000 1000>
    componentLinearAccScale=<1 1 1>
    componentLinearVelScale=<1 1 1>
    gravityOverride {z=-980}
    physicsBodyDefinitions [{
      boundBone {boneName=dyn_chest_rig_l}
      boxExtents=<22.6955 14.4129 4.8467>
      constraintSetup {angularLimitsMax=<15 1 1>, angularLimitsMin=<-15 -1 -1>, angularTarget=<1 0 0>}
      localJointOffset=<2.3215 0 0.0172>
     }]
   }
   n10 = RigidBody(ComponentPose=@n9) {
    bOverrideWorldGravity=true
    overridePhysicsAsset=&/M_MED_SageTwig_Physics
    simulationSpace=BaseBoneSpace
    alphaBoolBlend {blendOption=Linear}
    baseBoneRef {boneName=pelvis}
    componentAppliedLinearAccClamp=<1000 1000 1000>
    componentLinearAccScale {z=1}
    componentLinearVelScale {z=1}
    overrideWorldGravity {z=-980}
    simSpaceSettings {maxAngularAcceleration=1, maxAngularVelocity=10, maxLinearAcceleration=7, maxLinearVelocity=100, worldAlpha=1, externalLinearDragV=<0.2 0.2>}
   }
   Output Pose(Result=Component To Local(ComponentPose=@n10)) {layerGroup=None}
  ]}
 EventGraph {nodes [
   n1 = GetPartSkeletalMeshComponent()
   DrawJointDebugFunction!(
    execute
    SkelMeshComponent:Scene Component Object Reference=@n1
    JointOffset:Vector="0, 0, 0"
    JointName:Array of Names=Make Array("[0]":Name=upperarm_l) -> (Array:Array of Names)
    Scale:double=10
    Duration:double=0
    Thickness:double=0.8
    DebugDraw:bool=true
    Foreground:EDrawDebugSceneDepthPriorityGroup Enum=Foreground
   ) -> (then)
   > DrawJointDebugFunction(
    SkelMeshComponent=@n1
    JointOffset="0, 0, 0"
    JointName=Make Array("[0]"=upperarm_twist_01_l)
    Scale=50
    Duration=0
    Thickness=0.7
    DebugDraw=true
    Foreground=Foreground
   )
   Event BlueprintUpdateAnimation()
  ]}
]
variables []
