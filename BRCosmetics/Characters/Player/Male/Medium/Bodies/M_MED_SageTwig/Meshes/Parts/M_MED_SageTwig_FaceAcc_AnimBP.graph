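# Animation Blueprint for the face accessory part of the male medium SageTwig player cosmetic.
# Copies the pose from the attached parent mesh, then runs the part's Control Rig, with linear blending toggled by HeadLimit.
# Each update sets CinePosing when no valid pawn owner exists.
# HeadLimit is on only when not emoting, not in the front end and not CinePosing, so the rig applies only outside those cases.
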AnimBlueprint /BRCosmetics/Characters/Player/Male/Medium/Bodies/M_MED_SageTwig/Meshes/Parts/M_MED_SageTwig_FaceAcc_AnimBP
target_skeleton=&/M_MED_SageTwig_FaceAcc_Skeleton
graphs [
 AnimGraph {nodes [
   n1 = Control Rig!(Source=Copy Pose From Mesh() {bUseAttachedParent=true}, bAlphaBoolEnabled:bool=Get HeadLimit() -> (HeadLimit:bool)) -> (Pose) {
    alphaInputType=Bool
    alphaBoolBlend {blendOption=Linear}
    controlRigAssetReference {blueprintRigClass=&/M_MED_SageTwig_FaceAcc_ControlRig.M_MED_SageTwig_FaceAcc_ControlRig_C}
   }
   Output Pose(Result=@n1) {layerGroup=None}
  ]}
 EventGraph {nodes [
   Event BlueprintUpdateAnimation()
   > n1 = Is Valid(InputObject=TryGetPawnOwner())
   n2 = Set CinePosing(execute=@n1.Is Not Valid, CinePosing:bool=true) -> (Output_Get:bool)
   n3 = Get bIsPlayingEmote()
   n4 = Get bIsInFrontEnd()
   n5 = Get CinePosing() -> (CinePosing:bool)
   Set HeadLimit(
    execute=[@n2.then, @n6.then]
    HeadLimit:bool=AND Boolean!(A=NOT Boolean(A=@n3), B=NOT Boolean(A=@n4), C:bool=NOT Boolean(A=@n5)) -> (ReturnValue)
   ) -> (Output_Get:bool)
   n6 = Set CinePosing(execute=@n1.Is Valid, CinePosing=false)
  ]}
]
variables [
 CinePosing {type=boolean}
 HeadLimit {type=boolean, default=true}
]
